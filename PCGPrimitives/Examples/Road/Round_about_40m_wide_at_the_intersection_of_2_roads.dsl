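; Example PCG graph that builds a roundabout where two roads cross, 40 m wide per the asset name.
; Two spline lines feed an intersection-point subgraph, whose output seeds the roundabout spline shape.
; That shape cuts each road spline, and both cut roads and the roundabout each receive a shape grammar definition.
; The roundabout uses a spline-mesh-specific road profile to follow its tight curve.
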(graph Round_about_40m_wide_at_the_intersection_of_2_roads
  (bind Subgraph_0 (Subgraph :Graph /PCGPrimitives/Primitives/Create/Create_Spline_line.Create_Spline_line))
  (bind Subgraph_1 (Subgraph :Graph /PCGPrimitives/Primitives/Create/Create_Spline_line.Create_Spline_line))
  (bind Subgraph_2 (Subgraph :Graph /PCGPrimitives/Primitives/Place/Place_Points_AtIntersections.Place_Points_AtIntersections))
  (bind Subgraph_3 (Subgraph :Graph /PCGPrimitives/Primitives/Create/Create_Spline_Shape.Create_Spline_Shape))
  (bind Subgraph_7 (Subgraph :Graph /PCGPrimitives/Primitives/Compose/Compose_Spline_Cut.Compose_Spline_Cut))
  (bind Subgraph_4 (Subgraph :Graph /PCGPrimitives/Primitives/Assign/Assign_ShapeGrammarDefinition.Assign_ShapeGrammarDefinition))
  (bind Subgraph_8 (Subgraph :Graph /PCGPrimitives/Primitives/Compose/Compose_Spline_Cut.Compose_Spline_Cut))
  (bind Subgraph_5 (Subgraph :Graph /PCGPrimitives/Primitives/Assign/Assign_ShapeGrammarDefinition.Assign_ShapeGrammarDefinition))
  (bind Subgraph_9 (Subgraph :Graph /PCGPrimitives/Primitives/Assign/Assign_ShapeGrammarDefinition.Assign_ShapeGrammarDefinition)) ; Using Spline MEsh specific Road  profile since we need it to follow the tight curve of the round about
  (connect Subgraph_0 Out Subgraph_2 PrimaryInput)
  (connect Subgraph_1 Out Subgraph_2 PrimaryInput)
  (connect Subgraph_2 Out Subgraph_3 PrimaryInput)
  (connect Subgraph_7 Out Subgraph_4 PrimaryInput)
  (connect Subgraph_8 Out Subgraph_5 PrimaryInput)
  (connect Subgraph_0 Out Subgraph_7 PrimaryInput)
  (connect Subgraph_3 Out Subgraph_7 SecondaryInput)
  (connect Subgraph_1 Out Subgraph_8 PrimaryInput)
  (connect Subgraph_3 Out Subgraph_8 SecondaryInput)
  (connect Subgraph_3 Out Subgraph_9 PrimaryInput))
