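; PCG example graph for a simple straight road, named as 288m long, 15m wide and running north-south.
; It chains two reusable PCGPrimitives subgraphs: Create_Spline_line produces a line spline, which feeds the primary input of Assign_ShapeGrammarDefinition.
; The graph itself shows no dimension overrides, meshes or connection to the graph output.
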(graph Simple_Straight_Road_288m_long_15m_wide_North_South_bound
  (bind Subgraph_1 (Subgraph :Graph /PCGPrimitives/Primitives/Create/Create_Spline_line.Create_Spline_line))
  (bind Subgraph_2 (Subgraph :Graph /PCGPrimitives/Primitives/Assign/Assign_ShapeGrammarDefinition.Assign_ShapeGrammarDefinition))
  (connect Subgraph_1 Out Subgraph_2 PrimaryInput))
