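; PCG example graph that lays out a straight road with a sidewalk on each side, built from reusable PCGPrimitives subgraphs.
; A Create_Spline_line subgraph supplies the base line, which feeds a shape grammar assignment directly and two Transform_Offset branches.
; Each offset line gets its own shape grammar assignment, giving three parallel tagged lanes from one source spline.
; The graph connects nothing downstream of these three assignments.
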(graph Simple_Straight_with_sidewalk_on_each_side
  (bind Subgraph_1 (Subgraph :Graph /PCGPrimitives/Primitives/Create/Create_Spline_line.Create_Spline_line))
  (bind Subgraph_2 (Subgraph :Graph /PCGPrimitives/Primitives/Assign/Assign_ShapeGrammarDefinition.Assign_ShapeGrammarDefinition))
  (bind Subgraph_0 (Subgraph :Graph /PCGPrimitives/Primitives/Transform/Transform_Offset.Transform_Offset))
  (bind Subgraph_3 (Subgraph :Graph /PCGPrimitives/Primitives/Assign/Assign_ShapeGrammarDefinition.Assign_ShapeGrammarDefinition))
  (bind Subgraph_4 (Subgraph :Graph /PCGPrimitives/Primitives/Transform/Transform_Offset.Transform_Offset))
  (bind Subgraph_5 (Subgraph :Graph /PCGPrimitives/Primitives/Assign/Assign_ShapeGrammarDefinition.Assign_ShapeGrammarDefinition))
  (connect Subgraph_1 Out Subgraph_0 PrimaryInput)
  (connect Subgraph_1 Out Subgraph_2 PrimaryInput)
  (connect Subgraph_0 Out Subgraph_3 PrimaryInput)
  (connect Subgraph_1 Out Subgraph_4 PrimaryInput)
  (connect Subgraph_4 Out Subgraph_5 PrimaryInput))
