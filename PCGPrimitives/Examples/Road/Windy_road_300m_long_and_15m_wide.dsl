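; PCG example graph that, as named, builds a windy road 300 m long and 15 m wide from chained primitive subgraphs.
; It creates a spline line, resamples it, then jitters it to produce the winding path.
; A shape grammar definition is assigned last, using the spline mesh road profile so the result follows the curvature.
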(graph Windy_road_300m_long_and_15m_wide
  (bind Subgraph_0 (Subgraph :Graph /PCGPrimitives/Primitives/Create/Create_Spline_line.Create_Spline_line))
  (bind Subgraph_1 (Subgraph :Graph /PCGPrimitives/Primitives/Transform/Transform_Resample.Transform_Resample))
  (bind Subgraph_2 (Subgraph :Graph /PCGPrimitives/Primitives/Transform/Transform_Jitter.Transform_Jitter))
  (bind Subgraph_3 (Subgraph :Graph /PCGPrimitives/Primitives/Assign/Assign_ShapeGrammarDefinition.Assign_ShapeGrammarDefinition)) ; using the spline mesh version of the road profile demo as it will provide better result with the winding road.\nand adapt to its curvature.
  (connect Subgraph_0 Out Subgraph_1 PrimaryInput)
  (connect Subgraph_1 Out Subgraph_2 PrimaryInput)
  (connect Subgraph_2 Out Subgraph_3 PrimaryInput))
